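# Editor Utility Widget giving environment artists quick viewport toggles for LOD, Nanite and collision inspection.
# LOD buttons run r.ForceLOD (forcing also disables Nanite), while enabling the Nanite toggle greys out and blocks the force buttons.
# Collision toggles switch view modes and relabel themselves; a ">>" button opens the widget's own Blueprint via Python.
# Construct shows SM5/SM6/Mobile buttons only for one hardcoded username, but those buttons have no logic wired.
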EditorUtilityWidgetBlueprint /Game/ArtTools/Environment/EnvImpQuickSettings/EUW_EnvImpQuickSettings
graphs [EventGraph {nodes [
   "On Clicked (BTN_ForceLOD0)"!() -> (OutputDelegate, then, widget:User Widget Object Reference, event:Pointer Event Structure) {is=title K2Node_ComponentBoundEvent}
   > ExecuteConsoleCommand(Command=r.nanite 0)
   > ExecuteConsoleCommand(Command=r.ForceLOD 0)
   "On Clicked (BTN_ForceLOD1)"!() -> (OutputDelegate, then, widget:User Widget Object Reference, event:Pointer Event Structure) {is=title K2Node_ComponentBoundEvent}
   > ExecuteConsoleCommand(Command=r.nanite 0)
   > ExecuteConsoleCommand(Command=r.ForceLOD 1)
   "On Clicked (BTN_ForceLOD2)"!() -> (OutputDelegate, then, widget:User Widget Object Reference, event:Pointer Event Structure) {is=title K2Node_ComponentBoundEvent}
   > ExecuteConsoleCommand(Command=r.nanite 0)
   > ExecuteConsoleCommand(Command=r.ForceLOD 2)
   "On Clicked (BTN_ForceLOD3)"!() -> (OutputDelegate, then, widget:User Widget Object Reference, event:Pointer Event Structure) {is=title K2Node_ComponentBoundEvent}
   > ExecuteConsoleCommand(Command=r.nanite 0)
   > ExecuteConsoleCommand(Command=r.ForceLOD 3)
   "On Clicked (BTN_ResetLOD)"!() -> (OutputDelegate, then, widget:User Widget Object Reference, event:Pointer Event Structure) {is=title K2Node_ComponentBoundEvent}
   > ExecuteConsoleCommand(Command=r.ForceLOD -1)
   "On Clicked (BTN_EditWidget)"!() -> (OutputDelegate, then, widget:User Widget Object Reference, event:Pointer Event Structure) {is=title K2Node_ComponentBoundEvent}
   > n1 = GetBlueprintAsset(Object=&self)
   > Execute Python Script!(
    execute
    PythonScript="import unreal\r\n\r\nmy_widget = widget_object\r\nasset_editor_subsystem = unreal.get_editor_subsystem(unreal.AssetEditorSubsystem)\r\nasset_editor_subsystem.open_editor_for_assets([my_widget])"
    widget_object:Blueprint Object Reference=@n1
   ) -> (then, ReturnValue) {inputs [widget_object]}
   n2 = "On Selection Changed (BTN_ToggleNanite)"!() -> (OutputDelegate, then, selected:bool, widget:User Widget Object Reference) {is=title K2Node_ComponentBoundEvent}
   > n3 = Branch(Condition=@n2.selected)
   > ExecuteConsoleCommand(Command=r.nanite 1)
   n4 = Make Array!(
    "[0]":TA Generic 01 Button Object Reference=Get BTN_ForceLOD0() -> (BTN_ForceLOD0:TA Generic 01 Button Object Reference)
    "[1]":TA Generic 01 Button Object Reference=Get BTN_ForceLOD1() -> (BTN_ForceLOD1:TA Generic 01 Button Object Reference)
    "[2]":TA Generic 01 Button Object Reference=Get BTN_ForceLOD2() -> (BTN_ForceLOD2:TA Generic 01 Button Object Reference)
    "[3]":TA Generic 01 Button Object Reference=Get BTN_ForceLOD3() -> (BTN_ForceLOD3:TA Generic 01 Button Object Reference)
   ) -> (Array:Array of TA Generic 01 Button Object References)
   > n5 = For Each Loop(Array:Array of TA Generic 01 Button Object References=@n4) -> (Array Element:TA Generic 01 Button Object Reference)
   > Set ColorAndOpacity!(execute, ColorAndOpacity:Linear Color Structure=(R=0.5029,G=0.5029,B=0.5029,A=1), self:User Widget Object Reference=@n5) -> (then, Output_Get:Linear Color Structure)
   > Set Visibility(Visibility=HitTestInvisible, self=@n5)
   ExecuteConsoleCommand(execute=@n3.else, Command=r.nanite 0)
   > n6 = For Each Loop(Array=@n4)
   > Set ColorAndOpacity(ColorAndOpacity=(R=1,G=1,B=1,A=1), self=@n6)
   > Set Visibility(Visibility=Visible, self=@n6)
   n7 = "On Selection Changed (BTN_ToggleCollisionView)"!() -> (OutputDelegate, then, selected:bool, widget:User Widget Object Reference) {is=title K2Node_ComponentBoundEvent}
   > n8 = Branch(Condition=@n7.selected)
   > SetEditorViewportViewMode!(execute, Index:EViewModeIndex Enum=VMI_CollisionPawn) -> (then)
   n9 = Get BTN_ToggleCollisionView() -> (BTN_ToggleCollisionView:TA Generic 01 Button Object Reference)
   > setLabel!(execute, self:TA Base Editor Button Widget Object Reference=@n9, label:String=Hide Simple Collision) -> (then)
   SetEditorViewportViewMode(execute=@n8.else, Index=VMI_Lit)
   > setLabel(self=@n9, label=Show Simple Collision)
   n10 = "On Selection Changed (BTN_ToggleCollisionWireframe)"!() -> (OutputDelegate, then, selected:bool, widget:User Widget Object Reference) {is=title K2Node_ComponentBoundEvent}
   > n11 = Branch(Condition=@n10.selected)
   > SetEditorViewportViewMode(Index=VMI_CollisionVisibility)
   n12 = Get BTN_ToggleCollisionWireframe() -> (BTN_ToggleCollisionWireframe:TA Generic 01 Button Object Reference)
   > setLabel(self=@n12, label=Hide Complex Collision)
   SetEditorViewportViewMode(execute=@n11.else, Index=VMI_Lit)
   > setLabel(self=@n12, label=Show Complex Collision)
   Event Construct()
   > n13 = Branch(Condition="Equal Exactly (String)"(A=GetPlatformUserName!() -> (ReturnValue:String), B=brianfasten))
   n14 = Make Array!(
    "[0]"=Get BTN_ShowDX11() -> (BTN_ShowDX11:TA Generic 01 Button Object Reference)
    "[1]"=Get BTN_ShowDX12() -> (BTN_ShowDX12:TA Generic 01 Button Object Reference)
    "[2]"=Get BTN_ShowMobilePreview() -> (BTN_ShowMobilePreview:TA Generic 01 Button Object Reference)
   ) -> (Array)
   > n15 = For Each Loop(Array=@n14)
   n16 = For Each Loop(Exec=@n13.else, Array=@n14)
   SetVisibility(execute=@n15.LoopBody, self=[@n16, @n15], InVisibility=Visible)
   SetVisibility(execute=@n16.LoopBody, self=[@n16, @n15], InVisibility=Hidden)
   "On Clicked (BTN_ShowDX11)"!() -> (OutputDelegate, then, widget:User Widget Object Reference, event:Pointer Event Structure) {is=title K2Node_ComponentBoundEvent}
   Event PreConstruct()
   Event Tick()
   Get BTN_ToggleCollisionWireframe()
   Get BTN_ToggleCollisionView()
  ]}]
variables []
widgets {
 info {rootWidgetClass=&VerticalBox, widgetCount=17}
 widgets [
  VerticalBox_1 {widgetClassPath=&VerticalBox, widget=VerticalBox_1}
  HorizontalBox_0 {widgetClassPath=&HorizontalBox, parent=VerticalBox_1, widget=HorizontalBox_0, slot {class=VerticalBoxSlot}}
  VerticalBox_0 {
   widgetClassPath=&VerticalBox
   parent=HorizontalBox_0
   widget=VerticalBox_0
   slot {class=HorizontalBoxSlot, properties {padding=<ltrb 5 5 5 5>}}
  }
  BTN_ForceLOD0 {
   widgetClassPath=&TA_Generic01_Button
   parent=VerticalBox_0
   widget=BTN_ForceLOD0
   properties {label=Force LOD0}
   slot {class=VerticalBoxSlot, properties {padding=<ltrb 5 5 5 5>}}
  }
  BTN_ForceLOD1 {
   widgetClassPath=&TA_Generic01_Button
   parent=VerticalBox_0
   widget=BTN_ForceLOD1
   properties {label=Force LOD1}
   slot {class=VerticalBoxSlot, properties {padding=<ltrb 5 5 5 5>}}
  }
  BTN_ForceLOD2 {
   widgetClassPath=&TA_Generic01_Button
   parent=VerticalBox_0
   widget=BTN_ForceLOD2
   properties {label=Force LOD2}
   slot {class=VerticalBoxSlot, properties {padding=<ltrb 5 5 5 5>}}
  }
  BTN_ForceLOD3 {
   widgetClassPath=&TA_Generic01_Button
   parent=VerticalBox_0
   widget=BTN_ForceLOD3
   properties {label=Force LOD3}
   slot {class=VerticalBoxSlot, properties {padding=<ltrb 5 5 5 5>}}
  }
  BTN_ResetLOD {
   widgetClassPath=&TA_Generic01_Button
   parent=VerticalBox_0
   widget=BTN_ResetLOD
   properties {label=Reset LOD Preview}
   slot {class=VerticalBoxSlot, properties {padding=<ltrb 5 5 5 5>}}
  }
  VerticalBox {
   widgetClassPath=&VerticalBox
   parent=HorizontalBox_0
   widget=VerticalBox
   slot {class=HorizontalBoxSlot, properties {padding=<ltrb 5 5 5 5>}}
  }
  BTN_ToggleCollisionView {
   widgetClassPath=&TA_Generic01_Button
   parent=VerticalBox
   widget=BTN_ToggleCollisionView
   properties {autoToggleSelection=true, label=Show Simple Collision}
   slot {class=VerticalBoxSlot, properties {padding=<ltrb 5 5 5 5>}}
  }
  BTN_ToggleCollisionWireframe {
   widgetClassPath=&TA_Generic01_Button
   parent=VerticalBox
   widget=BTN_ToggleCollisionWireframe
   properties {autoToggleSelection=true, label=Show Complex Collision}
   slot {class=VerticalBoxSlot, properties {padding=<ltrb 5 5 5 5>}}
  }
  BTN_ShowDX11 {
   widgetClassPath=&TA_Generic01_Button
   parent=VerticalBox
   widget=BTN_ShowDX11
   properties {label=View SM5}
   slot {class=VerticalBoxSlot, properties {padding=<ltrb 5 5 5 5>}}
  }
  BTN_ShowDX12 {
   widgetClassPath=&TA_Generic01_Button
   parent=VerticalBox
   widget=BTN_ShowDX12
   properties {label=View SM6}
   slot {class=VerticalBoxSlot, properties {padding=<ltrb 5 5 5 5>}}
  }
  BTN_ShowMobilePreview {
   widgetClassPath=&TA_Generic01_Button
   parent=VerticalBox
   widget=BTN_ShowMobilePreview
   properties {label=View Mobile}
   slot {class=VerticalBoxSlot, properties {padding=<ltrb 5 5 5 5>}}
  }
  BTN_EditWidget {
   widgetClassPath=&TA_Generic01_Button
   parent=HorizontalBox_0
   widget=BTN_EditWidget
   properties {label=">>", buttonColor_Default=<rgba 0.0469 0.0469 0.0469 1>, sizeOverride=<20 0>}
   slot {class=HorizontalBoxSlot, properties {padding=<ltrb 5 5 5 5>}}
  }
  HorizontalBox_1 {
   widgetClassPath=&HorizontalBox
   parent=VerticalBox_1
   widget=HorizontalBox_1
   slot {class=VerticalBoxSlot, properties {padding=<ltrb 5 5 5 5>}}
  }
  BTN_ToggleNanite {
   widgetClassPath=&TA_Generic01_Button
   parent=HorizontalBox_1
   widget=BTN_ToggleNanite
   properties {autoToggleSelection=true, label=Toggle Nanite, sizeOverride=<350 0>}
   slot {class=HorizontalBoxSlot, properties {padding=<ltrb 10 0 0 0>}}
  }
 ]
}
